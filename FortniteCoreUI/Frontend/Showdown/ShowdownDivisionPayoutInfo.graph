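# Showdown frontend widget that displays league and division payouts, built on the native FortDivisionPayoutInfo class.
# League and division title texts sit above a scrollable wrap box of DivisionInfoPayoutEntry widgets.
# A bottom row holds dynamic DivisionIconButton entries between left and right arrow buttons bound to change-division input actions.
# The Blueprint supplies only layout and styling through inherited named widgets, leaving poster, timer and debug slots unplaced; its event graph is empty.
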WidgetBlueprint /FortniteCoreUI/Frontend/Showdown/ShowdownDivisionPayoutInfo
graphs [EventGraph {links [], nodes []}]
variables []
widgets {
 info {inheritedWidgetCount=13, parentClass=/Script/FortniteUI.FortDivisionPayoutInfo, rootWidgetClass=&Overlay, widgetCount=19}
 widgets [
  Overlay_0 {widgetClassPath=&Overlay, widget=Overlay_0}
  VerticalBox_2 {
   widgetClassPath=&VerticalBox
   parent=Overlay_0
   widget=VerticalBox_2
   slot {class=OverlaySlot, properties {horizontalAlignment=HAlign_Fill, verticalAlignment=VAlign_Fill, padding=<ltrb 0 -50 0 0>}}
  }
  Text_League {
   widgetClassPath=&CommonTextBlock
   bInherited=true
   bIsVariable=true
   parent=VerticalBox_2
   widget=Text_League
   properties {
    justification=Center
    text=Open League
    textTransformPolicy=ToUpper
    font {
     bForceMonospaced=false
     bMaterialIsStencil=false
     fontMaterial=null
     fontObject=&Burbank
     letterSpacing=0
     monospacedWidth=1
     size=60
     skewAmount=0
     typefaceFontName=Burbank-Black
     outlineSettings {
      bApplyOutlineToDropShadows=false
      bMiteredCorners=false
      bSeparateFillAlpha=false
      outlineMaterial=null
      outlineSize=0
      outlineColor=<rgba 0 0 0 1>
     }
    }
    shadowColorAndOpacity=<rgba 0.0273 0.0012 0.1221 1>
    shadowOffset=<8 8>
   }
   slot {class=VerticalBoxSlot, properties {horizontalAlignment=HAlign_Center, padding=<ltrb 0 0 0 -25>}}
  }
  Text_DivsionRank {
   widgetClassPath=&CommonTextBlock
   bInherited=true
   bIsVariable=true
   parent=VerticalBox_2
   widget=Text_DivsionRank
   properties {
    justification=Center
    text=Division 1 : Solo
    textTransformPolicy=ToUpper
    colorAndOpacity {colorUseRule=UseColor_Specified, specifiedColor=<rgba 0 1 0.9823 1>}
    font {
     bForceMonospaced=false
     bMaterialIsStencil=false
     fontMaterial=null
     fontObject=&BurbankSmall
     letterSpacing=0
     monospacedWidth=1
     size=28
     skewAmount=0
     typefaceFontName=Black
     outlineSettings {
      bApplyOutlineToDropShadows=false
      bMiteredCorners=false
      bSeparateFillAlpha=false
      outlineMaterial=null
      outlineSize=0
      outlineColor=<rgba 0 0 0 1>
     }
    }
    shadowColorAndOpacity=<rgba 0.0273 0.0012 0.1221 1>
    shadowOffset=<8 8>
   }
   slot {class=VerticalBoxSlot, properties {horizontalAlignment=HAlign_Center}}
  }
  ScrollBox_Payouts {
   widgetClassPath=&ScrollBox
   bInherited=true
   parent=VerticalBox_2
   widget=ScrollBox_Payouts
   slot {class=VerticalBoxSlot, properties {horizontalAlignment=HAlign_Center, padding=<ltrb 0 50 0 0>, size {sizeRule=Fill, value=5}}}
  }
  SizeBox_0 {
   widgetClassPath=&SizeBox
   parent=ScrollBox_Payouts
   widget=SizeBox_0
   properties {bOverride_WidthOverride=true, widthOverride=1569.4292}
   slot {class=ScrollBoxSlot}
  }
  EntryBox_Payouts {
   widgetClassPath=&DynamicEntryBox
   bInherited=true
   bIsVariable=true
   parent=SizeBox_0
   widget=EntryBox_Payouts
   properties {
    entryBoxType=Wrap
    entryWidgetClass=/FortniteCoreUI/UI/Events/DivisionInfoPayoutEntry.DivisionInfoPayoutEntry_C
    numDesignerPreviewEntries=8
    entrySpacing=<200 10>
   }
   slot {class=SizeBoxSlot}
  }
  HorizontalBox_2 {
   widgetClassPath=&HorizontalBox
   parent=VerticalBox_2
   widget=HorizontalBox_2
   slot {class=VerticalBoxSlot, properties {horizontalAlignment=HAlign_Center, verticalAlignment=VAlign_Bottom, size {sizeRule=Fill, value=1}}}
  }
  SizeBox_1 {
   widgetClassPath=&SizeBox
   parent=HorizontalBox_2
   widget=SizeBox_1
   properties {bOverride_HeightOverride=true, bOverride_WidthOverride=true, heightOverride=44, widthOverride=44}
   slot {class=HorizontalBoxSlot, properties {horizontalAlignment=HAlign_Left, verticalAlignment=VAlign_Center, padding=<ltrb 0 0 50 0>}}
  }
  Button_Left {
   widgetClassPath=&/CircleArrowButton
   bInherited=true
   bIsVariable=true
   parent=SizeBox_1
   widget=Button_Left
   properties {
    bHideInputActionWithKeyboard=true
    flip=true
    onButtonBaseClicked=(&self:WidgetTree.Button_Left.HandleOnButtonClicked)
    onButtonBaseDoubleClicked=(&self:WidgetTree.Button_Left.HandleOnButtonDoubleClicked)
    onButtonBaseHovered=(&self:WidgetTree.Button_Left.HandleOnButtonHovered)
    onButtonBaseUnhovered=(&self:WidgetTree.Button_Left.HandleOnButtonUnhovered)
    onSelectedChangedBase=(&self:WidgetTree.Button_Left.HandleOnSelectedChanged)
    hoveredSlateSoundOverride {resourceObject=/Game/Sounds/Fort_UI_Sounds/FrontEnd/Fort_Athena_DirectAcquisition_Hover_Cue_02}
    pressedSlateSoundOverride {resourceObject=/Game/Sounds/Fort_UI_Sounds/FrontEnd/Fort_Athena_LockerItem_Pressed_02}
    triggeringInputAction {dataTable=&InputActionDataTable, rowName=ChangeLeagueDivisionLeft}
   }
   slot {class=SizeBoxSlot}
  }
  EntryBoxes_DivisionRanksButtons {
   widgetClassPath=&DynamicEntryBox
   bInherited=true
   bIsVariable=true
   parent=HorizontalBox_2
   widget=EntryBoxes_DivisionRanksButtons
   properties {entryWidgetClass=/FortniteCoreUI/UI/Events/DivisionIconButton.DivisionIconButton_C, numDesignerPreviewEntries=10, visibility=Visible}
   slot {class=HorizontalBoxSlot, properties {horizontalAlignment=HAlign_Center, verticalAlignment=VAlign_Center}}
  }
  SizeBox_2 {
   widgetClassPath=&SizeBox
   parent=HorizontalBox_2
   widget=SizeBox_2
   properties {bOverride_HeightOverride=true, bOverride_WidthOverride=true, heightOverride=44, widthOverride=44}
   slot {class=HorizontalBoxSlot, properties {horizontalAlignment=HAlign_Right, verticalAlignment=VAlign_Center, padding=<ltrb 50 0 0 0>}}
  }
  Button_Right {
   widgetClassPath=&CircleArrowButton
   bInherited=true
   bIsVariable=true
   parent=SizeBox_2
   widget=Button_Right
   properties {
    bHideInputActionWithKeyboard=true
    onButtonBaseClicked=(&self:WidgetTree.Button_Right.HandleOnButtonClicked)
    onButtonBaseDoubleClicked=(&self:WidgetTree.Button_Right.HandleOnButtonDoubleClicked)
    onButtonBaseHovered=(&self:WidgetTree.Button_Right.HandleOnButtonHovered)
    onButtonBaseUnhovered=(&self:WidgetTree.Button_Right.HandleOnButtonUnhovered)
    onSelectedChangedBase=(&self:WidgetTree.Button_Right.HandleOnSelectedChanged)
    hoveredSlateSoundOverride {resourceObject=/Game/Sounds/Fort_UI_Sounds/FrontEnd/New_FrontEnd/Fortnite_UI_Map_Hover_04_SC_Cue}
    pressedSlateSoundOverride {resourceObject=&Fort_Athena_LockerItem_Pressed_02}
    triggeringInputAction {dataTable=&InputActionDataTable, rowName=ChangeLeagueDivisionRight}
   }
   slot {class=SizeBoxSlot}
  }
  Text_TimeRemaining {widgetClassPath=&CommonTextBlock, bInherited=true, widget=null}
  Text_DebugId {widgetClassPath=&TextBlock, bInherited=true, widget=null, properties=~}
  Image_PosterFront {widgetClassPath=/Script/EpicCMSUIFramework.EpicCMSImage, bInherited=true, widget=null}
  Image_PosterBack {widgetClassPath=&EpicCMSImage, bInherited=true, widget=null}
  Image_LoadingScreen {widgetClassPath=&EpicCMSImage, bInherited=true, widget=null}
  Image_PlaylistTile {widgetClassPath=&EpicCMSImage, bInherited=true, widget=null}
 ]
}
